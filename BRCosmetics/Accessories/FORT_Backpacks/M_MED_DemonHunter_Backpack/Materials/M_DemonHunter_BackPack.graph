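# Animated, glowing eyeball material for the Demon Hunter backpack cosmetic.
# Time-driven curve atlas lookups rotate the eye about local X and Z through switchable World Position Offset with corrected normals, and scale the pupil UVs.
# FX texture channels drive specular, roughness and tinted emissive, combined with a radial UV glow.
# Attributes pass through the accessories tech-art override function, with an optional eye-bulge normal feeding clear-coat normal and tangent outputs.
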Material /BRCosmetics/Accessories/FORT_Backpacks/M_MED_DemonHunter_Backpack/Materials/M_DemonHunter_BackPack
graph {
 properties {
  bUseMaterialAttributes=true
  bUsedWithMeshDeformer=true
  bUsedWithMorphTargets=true
  bUsedWithSkeletalMesh=true
  bUsedWithStaticLighting=true
  pixelDepthOffsetMode=PDOM_Legacy
  subsurfaceProfile=/Game/Characters/Shared/EyeMaterials/SSProfile/SSP_Eye_Inner
  usedShadingModels=Default Lit
 }
 nodes [
  n1 = Normalize(VectorInput=Transform(None=Constant3Vector() {constant=<rgba 1 0 0 0>}) {transformSourceType=TRANSFORMSOURCE_Local})
  n2 = Time()
  n3 = ScalarParameter() {defaultValue=0.1, parameterName=EyeMovementSpeed}
  n4 = CurveAtlasRowParameter(CurveTime=Frac(None=Multiply(A=@n2, B=@n3))) {
   atlas=&AthenaCurveAtlas01
   curve=/Game/Characters/Enemies/Cattus_Head/Materials/Cattus_EyeMovement_Curve
   defaultValue=92
  }
  n5 = MaterialFunctionCall(Float3=Subtract(A=@n4, B=Constant3Vector() {constant=<rgba 0.5 0.5 0.5 0>})) {materialFunction=&BreakOutFloat3Components}
  n6 = Multiply(A=ScalarParameter() {defaultValue=0.0476, parameterName=MaxRotationAmountX}, B=@n5.R)
  n7 = Normalize(VectorInput=Transform(None=Constant3Vector() {constant=<rgba 0 0 1 0>}) {transformSourceType=TRANSFORMSOURCE_Local})
  n8 = Multiply(A=ScalarParameter() {defaultValue=0.1, parameterName=MaxRotationAmountY}, B=@n5.G)
  n9 = MaterialFunctionCall(
   BaseNormal=VertexInterpolator(VS=MaterialFunctionCall(Rotation Axis=@n1, Rotation Angle=@n6) {materialFunction=&FixRotateAboutAxisNormals}.New Tangent Space Vertex Normal).PS
   AdditionalNormal=VertexInterpolator(VS=MaterialFunctionCall(Rotation Axis=@n7, Rotation Angle=@n8) {materialFunction=&FixRotateAboutAxisNormals}.New Tangent Space Vertex Normal).PS
  ) {materialFunction=&BlendAngleCorrectedNormals}
  n10 = StaticSwitchParameter(
   True=TextureSample() {samplerType=SAMPLERTYPE_Normal, texture=/Game/Characters/Shared/EyeMaterials/Textures/T_Eye_N}.RGB
   False=Constant3Vector() {constant=<rgba 0 0 1 0>}
  ) {defaultValue=true, parameterName=UseEyeBuldge}
  n11 = Constant3Vector() {constant=<rgba 1 1 1 0>}
  ClearCoatNormalCustomOutput!(None=MaterialFunctionCall(BaseNormal=@n9.Result, AdditionalNormal=Multiply(A=@n10, B=@n11) {constB=-1}) {materialFunction=&BlendAngleCorrectedNormals}.Result)
  n12 = Divide(A=Distance(A=TextureCoordinate(), B=Constant2Vector() {g=0.5, r=0.5}), B=ScalarParameter() {defaultValue=0.2714, parameterName=GlowRadius})
  Divide(A=Constant() {r=1}, B=Multiply(A=@n12, B=@n12))
  n13 = Time()
  n14 = ScalarParameter() {defaultValue=0.1, parameterName=EyeIrisadjustSpeed}
  n15 = CurveAtlasRowParameter(CurveTime=Frac(None=Multiply(A=@n13, B=@n14))) {
   atlas=&AthenaCurveAtlas01
   curve=&Cattus_EyeMovement_Curve
   defaultValue=92
  }
  n16 = MaterialFunctionCall(Texture Scale=LinearInterpolate(
    A=ScalarParameter() {defaultValue=0.9, parameterName=PupilScaleMin, sliderMax=1}
    B=ScalarParameter() {defaultValue=1.01, parameterName=PupilScaleMax, sliderMax=1}
    Alpha=@n15.B
   )) {materialFunction=&ScaleUVsByCenter}
  n17 = TextureSample(UVs=@n16.UVs) {texture=/BRCosmetics/Accessories/FORT_Backpacks/M_MED_DemonHunter_Backpack/Textures/T_M_MED_DemonHunter_Eyeball_Backpack_FX}
  n18 = OneMinus(None=@n17.G)
  n19 = ScalarParameter() {defaultValue=0.5, parameterName=Specular}
  n20 = ScalarParameter() {parameterName=Rougness}
  n21 = ScalarParameter() {defaultValue=10, parameterName=Rougness2}
  n22 = Add(A=@n17.R, B=Power(Base=@n17.R, Exp=ScalarParameter() {defaultValue=3, parameterName=Contrast}))
  n23 = ScalarParameter() {defaultValue=2, parameterName=Glow}
  n24 = OneMinus(None=@n12)
  n25 = ScalarParameter() {defaultValue=3, parameterName=contrast}
  n26 = Add(
   A=Multiply(A=Multiply(A=@n22, B=@n23), B=VectorParameter() {parameterName=Emissive, defaultValue=<rgba 0.1533 0.9323 0.4666 1>}.RGB)
   B=Multiply(A=VectorParameter() {parameterName=Emissive2, defaultValue=<rgba 0.0676 0.192 0.0811 1>}.RGB, B=Power(Base=@n24, Exp=@n25))
  )
  n27 = Multiply(A=@n17.B, B=VectorParameter() {parameterName=Emissive3, defaultValue=<rgba 1 0.0228 0.904 1>}.RGB)
  n28 = Add(
   A=Transform(None=VectorParameter() {parameterName=CenterLocation, defaultValue=<rgba 0 -10.5 0 1>}.RGB) {transformSourceType=TRANSFORMSOURCE_Local}
   B=MaterialFunctionCall() {materialFunction=&ObjectPivotPoint}.Object Pivot Location
  )
  n29 = WorldPosition()
  n30 = RotateAboutAxis(NormalizedRotationAxis=@n1, RotationAngle=@n6, PivotPoint=@n28, Position=@n29.XYZ)
  n31 = Add(A=@n30, B=RotateAboutAxis(NormalizedRotationAxis=@n7, RotationAngle=@n8, PivotPoint=@n28, Position=Add(A=@n30, B=@n29.XYZ)))
  n32 = Constant3Vector()
  MaterialFunctionCall(In=MakeMaterialAttributes!(
    BaseColor=TextureSample(UVs=@n16.UVs) {texture=/BRCosmetics/Accessories/FORT_Backpacks/M_MED_DemonHunter_Backpack/Textures/T_M_MED_DemonHunter_Eyeball_Backpack_D}.RGB
    Metallic
    Specular=Multiply(A=@n18, B=@n19)
    Roughness=LinearInterpolate(A=@n20, B=@n21, Alpha=@n17.G)
    Anisotropy
    EmissiveColor=Add(A=@n26, B=@n27)
    Opacity
    OpacityMask
    Normal=@n9.Result
    Tangent
    WorldPositionOffset=StaticSwitchParameter(True=@n31, False=@n32) {defaultValue=true, parameterName=UseWPOeyeMovement}
    SubsurfaceColor
    ClearCoat
    ClearCoatRoughness
    AmbientOcclusion
    Refraction
    CustomizedUVs_0
    CustomizedUVs_1
    CustomizedUVs_2
    CustomizedUVs_3
    CustomizedUVs_4
    CustomizedUVs_5
    CustomizedUVs_6
    CustomizedUVs_7
    PixelDepthOffset
    ShadingModel
    Displacement
   )) {materialFunction=&MF_TechArtOverrides_accessories}
  n33 = Transform(None=MaterialFunctionCall(BaseNormal=@n9.Result, AdditionalNormal=@n10) {materialFunction=&BlendAngleCorrectedNormals}.Result)
  TangentOutput!(None=Multiply(A=Normalize(VectorInput=@n33)))
 ]
}
