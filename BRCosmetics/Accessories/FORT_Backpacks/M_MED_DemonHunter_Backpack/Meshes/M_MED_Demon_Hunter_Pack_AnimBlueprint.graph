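# Animation Blueprint for the Demon Hunter backpack, which tilts the pack and makes its eye blink.
# The event graph maps the owner's spine_05 yaw relative to spine_01 to a clamped pack roll.
# A wrapping time counter, scaled by BlinkPerSecond, periodically opens a short window that produces a sine-eased blink weight, otherwise zero.
# The AnimGraph chains Modify Bone nodes from ref pose, using that weight as alpha for eyelid and eyeball bones.
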AnimBlueprint /BRCosmetics/Accessories/FORT_Backpacks/M_MED_DemonHunter_Backpack/Meshes/M_MED_Demon_Hunter_Pack_AnimBlueprint
target_skeleton=&/M_MED_Demon_Hunter_Pack_Skeleton
graphs [
 AnimGraph {nodes [
   n1 = "Transform (Modify) Bone - Bone: pack"(ComponentPose=Local To Component(LocalPose=Local Space Ref Pose()), Translation="0,0,0", Rotation=Get backpack_rotate()) {rotationMode=BMM_Additive, rotationSpace=BCS_BoneSpace, alphaBoolBlend {blendOption=Linear}, boneToModify {boneName=pack}}
   n2 = Get ModifyValue() -> (ModifyValue:double)
   n3 = "Transform (Modify) Bone - Bone: dyn_eyelid_up"(ComponentPose=@n1, Alpha=@n2, Translation="0.5,0.000,0", Rotation="-58,0,0", Scale="1,1,1") {
    rotationMode=BMM_Additive
    rotationSpace=BCS_BoneSpace
    translationMode=BMM_Additive
    translationSpace=BCS_BoneSpace
    alphaBoolBlend {blendOption=Linear}
    boneToModify {boneName=dyn_eyelid_up}
    rotation {pitch=-58}
    translation {x=0.5}
   }
   n4 = "Transform (Modify) Bone - Bone: dyn_eyelid_up_end"(ComponentPose=@n3, Alpha=@n2, Translation="1,0,1.5", Rotation="0,0,0", Scale="1,1,1") {
    translationMode=BMM_Additive
    translationSpace=BCS_BoneSpace
    alphaBoolBlend {blendOption=Linear}
    boneToModify {boneName=dyn_eyelid_up_end}
    translation {x=1, z=1.5}
   }
   n5 = "Transform (Modify) Bone - Bone: dyn_eyelid_dn"(ComponentPose=@n4, Alpha=@n2, Translation="0.5,0,0", Rotation="58,0,0", Scale="1,1,1") {
    rotationMode=BMM_Additive
    rotationSpace=BCS_BoneSpace
    alphaBoolBlend {blendOption=Linear}
    boneToModify {boneName=dyn_eyelid_dn}
    rotation {pitch=58}
    translation {x=0.5}
   }
   n6 = "Transform (Modify) Bone - Bone: dyn_eyelid_dn_end"(ComponentPose=@n5, Alpha=@n2, Translation="1,0,-1.5", Rotation="0,0,0", Scale="1,1,1") {
    translationMode=BMM_Additive
    translationSpace=BCS_BoneSpace
    alphaBoolBlend {blendOption=Linear}
    boneToModify {boneName=dyn_eyelid_dn_end}
    translation {x=1, z=-1.5}
   }
   n7 = "Transform (Modify) Bone - Bone: dyn_eyeball"(ComponentPose=@n6, Alpha=@n2, Translation="0,0,0", Rotation="0,0,0", Scale="0.95,0.95,0.95") {
    scaleMode=BMM_Replace
    translationSpace=BCS_BoneSpace
    alphaBoolBlend {blendOption=Linear}
    boneToModify {boneName=dyn_eyeball}
    scale=<0.95 0.95 0.95>
   }
   Output Pose(Result=Component To Local(ComponentPose=@n7)) {layerGroup=None}
  ]}
 EventGraph {nodes [
   n1 = Event BlueprintUpdateAnimation()
   > n2 = Sequence()
   n3 = GetOwnerSkeletalMeshComponent()
   n4 = BreakRotator(InRot=BreakTransform(InTransform=GetDeltaTransformFromRefPose(self=@n3, BoneName=spine_05, BaseName=spine_01)).Rotation)
   > Set backpack_rotate(backpack_rotate=MakeRotator(Roll=MapRangeClamped(Value=@n4.Yaw, InRangeA=0, InRangeB=-30, OutRangeB=20)))
   n5 = "Division (Whole and Remainder)"!(Dividend:double=Get BlinkX() -> (BlinkX:double), Divisor:double=15) -> (Remainder:double, ReturnValue:int=0)
   > n6 = Branch(Condition="float < float"!(A:double=@n5, B:double=1) -> (ReturnValue:bool) {is=title K2Node_CallFunction})
   n7 = "Sin (Degrees)"!(A:double=float * float(A=@n5, B=90)) -> (ReturnValue:double)
   > Set ModifyValue(ModifyValue:double="Sin (Degrees)"(A=FInterpEaseInOut!(A:double=0, B:double=180, Alpha:double=@n7, Exponent:double=1) -> (ReturnValue:double))) -> (Output_Get:double)
   Set ModifyValue(execute=@n6.else, ModifyValue=0)
   n8 = Branch(execute=@n2.then_1, Condition="float >= float"(A=Get BlinkX(), B=60) {is=title K2Node_CallFunction})
   > Set BlinkX(BlinkX:double=float - float(A=Get BlinkX(), B=60) {is=title K2Node_CallFunction}) -> (Output_Get:double)
   n9 = Get BlinkPerSecond() -> (BlinkPerSecond:double)
   Set BlinkX(
    execute=@n8.else
    BlinkX=float + float(A=Get BlinkX(), B=float * float(A=@n1.DeltaTimeX, B=@n9)) {is=title K2Node_CommutativeAssociativeBinaryOperator}
   )
  ]}
]
variables [
 backpack_rotate {type=Rotator}
 BlinkPerSecond {type=number, default=6}
 BlinkX {type=number}
 ModifyValue {type=number}
]
